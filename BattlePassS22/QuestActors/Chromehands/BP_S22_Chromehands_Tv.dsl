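; Interactable TV for the Season 22 battle pass Chromehands quest that reacts to whether the player wears the Bites outfit.
; Starting interaction swaps the screen material and ambient audio for interacting versions; interruption restores them, while completion resets the material and fades interaction audio.
; On interact, a deprecated loadout check selects the outfit voice-over or one of two random fallbacks, applied as gameplay effects.
; Locally, lacking the outfit restores ambience and sets CannotInteract, which overrides the failed-interaction text.
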(event Interaction|EventBlueprintOnInteract
  (params (InteractingPawn :type Fort%20Pawn%20Object%20Reference) (InteractionBeingAttempted :type EInteractionBeingAttempted%20Enum :default "FirstInteraction"))
  (bind _self self)
  (bind _returnvalue (Ability|GetAbilitySystemComponent _self))
  (bind _asfort_player_pawn (Utilities|Casting|CastToFortPlayerPawn InteractingPawn))
  (bind _returnvalue_2 (Ability|GetAbilitySystemComponent _asfort_player_pawn))
  (|CallParentFunction InteractingPawn InteractionBeingAttempted)
  (if (Class|FortCosmeticFunctionLibrary|%28Deprecated%29IsItemInLoadout (Variables|/Script/FortniteGame.FortPlayerPawn|GetCosmeticLoadout _asfort_player_pawn) (Variables|Default|GetCID_Bites))
    (GameplayEffects|ApplyGameplayEffectToTarget :self _returnvalue :GameplayEffectClass (Variables|Default|GetGE_PlayVOWithOutfit) :Target _returnvalue_2 :Level 0.0)
    (else
      (GameplayEffects|ApplyGameplayEffectToTarget :self _returnvalue :GameplayEffectClass (Utilities|Select :Index (Math|Random|RandomIntegerinRange 0 1) :Option%200 (Variables|Default|GetGE_NoOutfitVO_01) :Option%201 (Variables|Default|GetGE_NoOutfitVO_02)) :Target _returnvalue_2 :Level 0.0))))

(event Interaction|EventBlueprintOnBeginInteract
  (params (InteractingPawn :type Fort%20Pawn%20Object%20Reference) (InteractionBeingAttempted :type EInteractionBeingAttempted%20Enum :default "FirstInteraction") (InteractionType :type TInteractionType%20Enum :default "IT_NoInteraction"))
  (bind _audio_interacting (Variables|Default|GetAudio_Interacting))
  (|CallParentFunction InteractingPawn InteractionBeingAttempted InteractionType)
  (Rendering|Material|SetMaterial (Variables|Components|GetSM_Quest_PreInteraction) 1 (Variables|Default|GetMaterial_Interacting))
  (Components|Activation|Deactivate (Variables|Components|GetAmbientAudio))
  (Components|Activation|Activate _audio_interacting false)
  (Audio|Components|Audio|FadeIn _audio_interacting 0.2 1.0 0.0 "Linear" :!class /Script/Engine.AudioComponent))

(event Interaction|EventBlueprintOnInterruptInteract
  (bind _audio_interacting (Variables|Default|GetAudio_Interacting))
  (|CallParentFunction)
  (Rendering|Material|SetMaterial (Variables|Components|GetSM_Quest_PreInteraction) 1 (Variables|Default|GetMaterial_Ambient))
  (Components|Activation|Activate (Variables|Components|GetAmbientAudio) false)
  (Audio|Components|Audio|FadeOut _audio_interacting 0.2 0.0 "Linear" :!class /Script/Engine.AudioComponent)
  (Components|Activation|Deactivate _audio_interacting))

(event VisualState|VFX&Audio|EventPostInteractionVFX&Audio
  (params (PlayerController :type Fort%20Player%20Controller%20Object%20Reference))
  (bind _audio_interacting (Variables|Default|GetAudio_Interacting))
  (|CallParentFunction PlayerController)
  (Rendering|Material|SetMaterial (Variables|Components|GetSM_Quest_PreInteraction) 1 (Variables|Default|GetMaterial_Ambient))
  (Audio|Components|Audio|FadeOut _audio_interacting 0.2 0.0 "Linear" :!class /Script/Engine.AudioComponent)
  (Components|Activation|Deactivate _audio_interacting))

(fn BlueprintGetFailedInteractionString
  (params (InteractingPawn :type Fort%20Pawn%20Object%20Reference) (InteractionBeingAttempted :type EInteractionBeingAttempted%20Enum :default "FirstInteraction"))
  (bind _returnvalue (|CallParentFunction InteractingPawn InteractionBeingAttempted))
  (if (Variables|Default|GetCannotInteract)
    (return "NSLOCTEXT(\"[3836E0584CF3073AED37E29EC05FEAF7]\", \"ACFF12EB4041108DFC3F2691731E610D\", \"You\\'re not the one we\\'re looking for...\")")
    (else
      (return _returnvalue))))

(fn BlueprintOnLocalInteract
  (params (InteractingPawn :type Fort%20Player%20Pawn%20Object%20Reference) (InteractionBeingAttempted :type EInteractionBeingAttempted%20Enum :default "FirstInteraction"))
  (bind _audio_interacting (Variables|Default|GetAudio_Interacting))
  (bind _returnvalue (|CallParentFunction InteractingPawn InteractionBeingAttempted))
  (if (not (Class|FortCosmeticFunctionLibrary|%28Deprecated%29IsItemInLoadout (Variables|/Script/FortniteGame.FortPlayerPawn|GetCosmeticLoadout InteractingPawn) (Variables|Default|GetCID_Bites)))
    (Variables|Default|SetCannotInteract true)
    (Rendering|Material|SetMaterial (Variables|Components|GetSM_Quest_PreInteraction) 1 (Variables|Default|GetMaterial_Ambient))
    (Components|Activation|Activate (Variables|Components|GetAmbientAudio) false)
    (Audio|Components|Audio|FadeOut _audio_interacting 0.2 0.0 "Linear" :!class /Script/Engine.AudioComponent)
    (Components|Activation|Deactivate _audio_interacting))
  (return _returnvalue))

(fn ConstructionScript
  (|CallParentFunction))
